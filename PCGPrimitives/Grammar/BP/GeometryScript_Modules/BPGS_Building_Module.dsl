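; Geometry Script generated-mesh actor that builds one configurable building facade module, apparently for grammar-based PCG buildings.
; On rebuild it appends a body box, optional window and door boxes with their own material IDs, and an optional column cylinder.
; Corner modules get an angled left or right cut by boolean-subtracting a rotated box built on a compute mesh.
; It assigns three material slots and optionally bakes the result to a Static Mesh asset.
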(event EventBeginPlay)

(event Collision|EventActorBeginOverlap
  (params (OtherActor :type Actor%20Object%20Reference)))

(event EventTick
  (params (DeltaSeconds :type Float%20%28single-precision%29)))

(event Events|EventOnRebuildGeneratedMesh
  (params (TargetMesh :type Dynamic%20Mesh%20Object%20Reference))
  (bind _module_dimension_y (Variables|Module|GetModule%20Dimension%20Y))
  (bind _returnvalue (* _module_dimension_y -0.5))
  (bind _module_dimension_y_2 (Variables|Module|GetModule%20Dimension%20Y))
  (bind _returnvalue_2 (* _module_dimension_y_2 0.5))
  ;[1 1 1 1 100] MODULE BODY
  (bind _returnvalue_3 (GeometryScript|Primitives|AppendBox :TargetMesh TargetMesh :PrimitiveOptions (Utilities|Struct|MakeGeometryScriptPrimitiveOptions "PerFace") :DimensionX (Variables|Module|GetModule%20Dimension%20X) :DimensionY (Variables|Module|GetModule%20Dimension%20Y) :DimensionZ (Variables|Module|GetModule%20Dimension%20Z) :StepsX 0 :StepsY 0 :StepsZ 0 :Origin "Base"))
  (bind _returnvalue_4 (GeometryScript|Transforms|TransformMesh _returnvalue_3 (Math|Transform|MakeTransform (Math|Vector|MakeVector (/ (Variables|Module|GetModule%20Dimension%20X) -2.0)))))
  (Variables|Default|SetTarget%20Mesh _returnvalue_4)
  (Utilities|FlowControl|Sequence
    (:then_0
      ;[1 1 1 1 60] WINDOW
      (if (Variables|Window|GetAdd%20Window)
        (GeometryScript|Primitives|AppendBox (Variables|Default|GetTarget%20Mesh) (Utilities|Struct|MakeGeometryScriptPrimitiveOptions "PerFace" false "Uniform" 1) (Math|Transform|MakeTransform (Math|Vector|MakeVector 0.0 (Variables|Window|GetWindows%20Offset%20Y) (+ (- (/ (Variables|Module|GetModule%20Dimension%20Z) 2.0) (/ (Variables|Window|GetWindows%20Dimensions%20Z) 2.0)) (Variables|Window|GetWindows%20Offset%20Z)))) (Variables|Window|GetWindows%20Dimensions%20X) (Variables|Window|GetWindows%20Dimensions%20Y) (Variables|Window|GetWindows%20Dimensions%20Z))))
    (:then_1
      ;[1 1 1] DOOR
      (if (Variables|Door|GetAdd%20Door)
        (GeometryScript|Primitives|AppendBox :TargetMesh (Variables|Default|GetTarget%20Mesh) :PrimitiveOptions (Utilities|Struct|MakeGeometryScriptPrimitiveOptions "PerFace" false "Uniform" 2) :DimensionX (Variables|door|GetDoor%20Dimension%20X) :DimensionY (Variables|Door|GetDoor%20Dimension%20Y) :DimensionZ (Variables|Door|GetDoor%20Dimension%20Z) :StepsX 0 :StepsY 0 :StepsZ 0 :Origin "Base")))
    (:then_2
      (if (Variables|Module|GetCorner)
        (if (Variables|Module|GetIs%20right%20corner)
          ;[1 1 1 1 60] RIGHT CORNER
          (bind _returnvalue_5 (DynamicMeshActor|AllocateComputeMesh))
          (bind _returnvalue_6 (GeometryScript|Primitives|AppendBox :TargetMesh _returnvalue_5 :Transform (Math|Transform|MakeTransform (Math|Vector|MakeVector -5e+01 _returnvalue -2e+01)) :DimensionX 1e+02 :DimensionY 4e+02 :DimensionZ (* (Variables|Module|GetModule%20Dimension%20Z) 3.0) :StepsX 0 :StepsY 0 :StepsZ 0 :Origin "Base"))
          (GeometryScript|Transforms|RotateMesh _returnvalue_6 (Math|Vector|RotatorfromAxisandAngle "0.000000,0.000000,1.000000" 3e+01) (Math|Vector|MakeVector 0.0 _returnvalue))
          (GeometryScript|Booleans|ApplyMeshBoolean :TargetMesh (Variables|Default|GetTarget%20Mesh) :ToolMesh _returnvalue_6 :Operation "Subtract")
          (DynamicMeshActor|ReleaseAllComputeMeshes)
          (else
            ;[1 1 1 1 60] LEFT CORNER
            (bind _returnvalue_7 (DynamicMeshActor|AllocateComputeMesh))
            (bind _returnvalue_8 (GeometryScript|Primitives|AppendBox :TargetMesh _returnvalue_7 :Transform (Math|Transform|MakeTransform (Math|Vector|MakeVector -5e+01 _returnvalue_2 -2e+01)) :DimensionX 1e+02 :DimensionY 4e+02 :DimensionZ (* (Variables|Module|GetModule%20Dimension%20Z) 3.0) :StepsX 0 :StepsY 0 :StepsZ 0 :Origin "Base"))
            (GeometryScript|Transforms|RotateMesh _returnvalue_8 (Math|Vector|RotatorfromAxisandAngle "0.000000,0.000000,1.000000" -3e+01) (Math|Vector|MakeVector 0.0 _returnvalue_2))
            (GeometryScript|Booleans|ApplyMeshBoolean :TargetMesh (Variables|Default|GetTarget%20Mesh) :ToolMesh _returnvalue_8 :Operation "Subtract")
            (DynamicMeshActor|ReleaseAllComputeMeshes)))))
    (:then_3
      (Utilities|FlowControl|Sequence
        (:then_0
          ;[1 1 1 1 60] MATERIAL ASSIGNEMENT
          (Rendering|Material|SetMaterial (Actor|GetComponentbyClass "/Script/GeometryFramework.DynamicMeshComponent") 0 (Variables|Default|GetModule%20Material)))
        (:then_1
          (Rendering|Material|SetMaterial (Actor|GetComponentbyClass "/Script/GeometryFramework.DynamicMeshComponent") 1 (Variables|Default|GetWindows%20Material)))
        (:then_2
          (Rendering|Material|SetMaterial (Actor|GetComponentbyClass "/Script/GeometryFramework.DynamicMeshComponent") 2 (Variables|Default|GetDoor%20Material)))))
    (:then_4
      (if (Variables|Module|GetAdd%20Column)
        (GeometryScript|Primitives|AppendCylinder :TargetMesh (Variables|Default|GetTarget%20Mesh) :Radius (/ (Variables|Module|GetModule%20Dimension%20Y) 2.0) :Height (Variables|Module|GetModule%20Dimension%20Z) :RadialSteps 12 :HeightSteps 0 :bCapped true :Origin "Base")))
    (:then_5
      ;[1 1 1] BAKE Mesh
      (if (Variables|Default|GetBake%20Mesh)
        (bind _returnvalue_9 (GeometryScript|AssetManagement|CreateNewStaticMeshAssetfromMesh (Variables|Default|GetTarget%20Mesh) (Variables|Default|GetDestination) (Utilities|Struct|MakeGeometryScriptCreateNewStaticMeshAssetOptions true true true "(bEnabled=False,bExplicitTangents=False,bLerpUVs=True,bSeparable=False,bVoxelNDF=True,bVoxelOpacity=False,ShapePreservation=None,PositionPrecision=-2147483648,NormalPrecision=-1,TangentPrecision=-1,BoneWeightPrecision=-1,TargetMinimumResidencyInKB=0,KeepPercentTriangles=1.000000,TrimRelativeError=0.000000,GenerateFallback=PlatformDefault,FallbackTarget=Auto,FallbackPercentTriangles=1.000000,FallbackRelativeError=1.000000,MaxEdgeLengthFactor=0.000000,NumRays=64,VoxelLevel=0,RayBackUp=0.000000,DisplacementUVChannel=0,DisplacementMaps=,NaniteAssemblyData=(Parts=,Nodes=))" 0.0 false))
          (:Success
            (StaticMesh|SetMaterial _returnvalue_9 0 (Variables|Default|GetModule%20Material))
            (StaticMesh|SetMaterial _returnvalue_9 1 (Variables|Default|GetWindows%20Material))
            (StaticMesh|SetMaterial _returnvalue_9 2 (Variables|Default|GetDoor%20Material))))))))

(fn ConstructionScript)
